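# Sound cue for the Season 14 frontend main room ambience, playing a Battle Pass frontend ambient loop.
# A single looping wave player feeds the output at reduced volume, routed through the Fort_UI_InGame_Critical sound class.
# Overridden attenuation disables spatialization, and concurrency allows one instance at a time, blocking new ones while it plays.
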SoundCue /BRCosmetics/Sounds/Other/Fort_UI_Sounds/S14_Frontend/S14_AMB_MainRoom_Cue
root=WavePlayer_0
nodes [{
  id=WavePlayer_0
  properties {bLooping=true, soundWaveAssetPtr=&/S14_Battlepass_Frontend_Amb_Loop}
 }]
properties {
 bOverrideAttenuation=true
 maxDistance=1500
 soundClassObject=&Fort_UI_InGame_Critical
 volumeMultiplier=0.35
 attenuationOverrides {bSpatialize=false, falloffDistance=1000, stereoSpread=1000, attenuationShapeExtents {x=500}}
 concurrencyOverrides {maxCount=1, resolutionRule=PreventNew}
}
